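# Physics asset for the leather tail skirt on the medium female BoneMarrow player body.
# A two-link main chain hangs from a pelvis-locked base and forks into left and right three-link chains, with twist locked and swing limits widening toward the tips.
# Pelvis and thigh capsules, angularly locked to their parents, plus a large root capsule centered below the origin form body and likely ground collision shapes.
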PhysicsAsset /Game/Characters/Player/Female/Medium/Bodies/F_MED_BoneMarrow/Meshes/F_MED_BoneMarrow_Physics_Tail_skirt
bodies {
 base_sk_leather_main [{shapeType=Capsule, length=0.505, radius=0.505, shapeName=base_sk_leather_main_capsule, rotation=<rot 0 -90 0>}]
 dyn_sk_leather_01_l [{shapeType=Capsule, length=1, radius=4, shapeName=dyn_sk_leather_01_l_capsule, center=<5 0 0>, rotation=<rot -90 0 0>}]
 dyn_sk_leather_01_r [{shapeType=Capsule, length=1, radius=4, shapeName=dyn_sk_leather_01_r_capsule, center=<5 0 0>, rotation=<rot -90 0 0>}]
 dyn_sk_leather_02_l [{shapeType=Capsule, length=1, radius=4, shapeName=dyn_sk_leather_02_l_capsule, center=<6 0 0>, rotation=<rot -90 0 0>}]
 dyn_sk_leather_02_r [{shapeType=Capsule, length=1, radius=4, shapeName=dyn_sk_leather_02_r_capsule, center=<6 0 0>, rotation=<rot -90 0 0>}]
 dyn_sk_leather_03_l [{shapeType=Capsule, length=1, radius=2.25, shapeName=dyn_sk_leather_03_l_capsule, center=<3.5598 0 0>, rotation=<rot -90 180 180>}]
 dyn_sk_leather_03_r [{shapeType=Capsule, length=1, radius=2.25, shapeName=dyn_sk_leather_03_r_capsule, center=<3.5598 0 0>, rotation=<rot -90 180 180>}]
 dyn_sk_leather_main_01 [{shapeType=Capsule, length=1, radius=2.8, shapeName=dyn_sk_leather_main_01_capsule, center=<3.5 0 0>, rotation=<rot -90 0 0>}]
 dyn_sk_leather_main_02 [{shapeType=Capsule, length=1, radius=4.5, shapeName=dyn_sk_leather_main_02_capsule, center=<5 0 0>, rotation=<rot -90 0 0>}]
 pelvis [{shapeType=Capsule, length=12, radius=8.5, shapeName=pelvis_capsule, center=<-5.7077 -0.8628 0>}]
 root [{shapeType=Capsule, length=0, radius=62, shapeName=root_capsule, center=<0 0 -55>, rotation=<rot -90 -90 0>}]
 thigh_l [{shapeType=Capsule, length=5, radius=8.5, shapeName=thigh_r_capsule, center=<-22 -2.8 0>, rotation=<rot -90 -90.5672 90.8146>}]
 thigh_r [{shapeType=Capsule, length=5, radius=8.5, shapeName=thigh_r_capsule, center=<22 2.8 0>, rotation=<rot -90 180 180>}]
}
constraints [
 {
  bone1Name=thigh_r
  bone2Name=pelvis
  swing1LimitDegrees=45
  swing1Motion=Locked
  swing2LimitDegrees=45
  swing2Motion=Locked
  twistLimitDegrees=45
  twistMotion=Locked
 }
 {
  bone1Name=thigh_l
  bone2Name=pelvis
  swing1LimitDegrees=45
  swing1Motion=Locked
  swing2LimitDegrees=45
  swing2Motion=Locked
  twistLimitDegrees=45
  twistMotion=Locked
 }
 {
  bone1Name=dyn_sk_leather_main_01
  bone2Name=base_sk_leather_main
  swing1LimitDegrees=2
  swing2LimitDegrees=15
  twistLimitDegrees=0
  twistMotion=Locked
 }
 {
  bone1Name=dyn_sk_leather_main_02
  bone2Name=dyn_sk_leather_main_01
  swing1LimitDegrees=8
  swing2LimitDegrees=20
  twistLimitDegrees=0
  twistMotion=Locked
 }
 {
  bone1Name=dyn_sk_leather_01_l
  bone2Name=dyn_sk_leather_main_02
  swing1LimitDegrees=20
  swing2LimitDegrees=30
  twistLimitDegrees=0
  twistMotion=Locked
 }
 {
  bone1Name=dyn_sk_leather_02_l
  bone2Name=dyn_sk_leather_01_l
  swing1LimitDegrees=35
  swing2LimitDegrees=35
  twistLimitDegrees=0
  twistMotion=Locked
 }
 {
  bone1Name=dyn_sk_leather_03_l
  bone2Name=dyn_sk_leather_02_l
  swing1LimitDegrees=45
  swing2LimitDegrees=45
  twistLimitDegrees=0
  twistMotion=Locked
 }
 {
  bone1Name=dyn_sk_leather_01_r
  bone2Name=dyn_sk_leather_main_02
  swing1LimitDegrees=20
  swing2LimitDegrees=30
  twistLimitDegrees=0
  twistMotion=Locked
 }
 {
  bone1Name=dyn_sk_leather_02_r
  bone2Name=dyn_sk_leather_01_r
  swing1LimitDegrees=35
  swing2LimitDegrees=35
  twistLimitDegrees=0
  twistMotion=Locked
 }
 {
  bone1Name=dyn_sk_leather_03_r
  bone2Name=dyn_sk_leather_02_r
  swing1LimitDegrees=45
  swing2LimitDegrees=45
  twistLimitDegrees=0
  twistMotion=Locked
 }
 {
  bone1Name=base_sk_leather_main
  bone2Name=pelvis
  swing1LimitDegrees=45
  swing1Motion=Locked
  swing2LimitDegrees=45
  swing2Motion=Locked
  twistLimitDegrees=45
  twistMotion=Locked
 }
 {
  bone1Name=pelvis
  bone2Name=root
  swing1LimitDegrees=45
  swing1Motion=Locked
  swing2LimitDegrees=45
  swing2Motion=Locked
  twistLimitDegrees=45
  twistMotion=Locked
 }
]
