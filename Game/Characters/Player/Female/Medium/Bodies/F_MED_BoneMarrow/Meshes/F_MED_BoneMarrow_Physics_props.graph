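# Physics asset for the dangling props on the female medium BoneMarrow player body: ape toy, bottle, knife, feathers and right-side leather pieces.
# Tiny base capsules lock angularly to the pelvis and anchor limited-swing prop capsules, with the feathers forming a four-link, twist-locked chain.
# The upper leather piece rotates freely from the pelvis with the lower locked to it.
# Pelvis, lower spine and right-leg capsules supply body shapes.
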PhysicsAsset /Game/Characters/Player/Female/Medium/Bodies/F_MED_BoneMarrow/Meshes/F_MED_BoneMarrow_Physics_props
bodies {
 base_ape_toy [{shapeType=Capsule, length=0.505, radius=0.505, shapeName=base_ape_toy_capsule, rotation=<rot 0 -90 0>}]
 base_bottle [{shapeType=Capsule, length=0.505, radius=0.505, shapeName=base_bottle_capsule, rotation=<rot 0 -90 0>}]
 base_feather_1 [{shapeType=Capsule, length=0.505, radius=0.505, shapeName=base_feather_1_capsule, rotation=<rot 0 -90 0>}]
 base_knife [{shapeType=Capsule, length=0.505, radius=0.505, shapeName=base_knife_capsule, rotation=<rot 0 -90 0>}]
 calf_r [{shapeType=Capsule, length=41.5919, radius=4.9123, shapeName=calf_r_capsule, center=<20.2288 2.4931 0.45>, rotation=<rot 0 97.026 -88.7351>}]
 dyn_ape_toy [{shapeType=Capsule, length=0.75, radius=3, shapeName=dyn_ape_toy_capsule, center=<5.2294 0 0.7796>, rotation=<rot -90 180 180>}]
 dyn_bottle [{shapeType=Capsule, length=5.25, radius=4.25, shapeName=dyn_bottle_capsule, center=<6.4869 0 0>, rotation=<rot -90 0 0>}]
 dyn_feather_1 [{shapeType=Capsule, length=0.5, radius=1.5, shapeName=dyn_feather_1_capsule, center=<3.2902 0 0>, rotation=<rot -90 0 0>}]
 dyn_feather_2 [{shapeType=Capsule, length=4.25, radius=2.5, shapeName=dyn_feather_2_capsule, center=<3.1502 0 0>, rotation=<rot -90 0 0>}]
 dyn_feather_3 [{shapeType=Capsule, length=4.5, radius=2.5, shapeName=dyn_feather_3_capsule, center=<5.1454 0 0>, rotation=<rot -90 180 180>}]
 dyn_feather_4 [{shapeType=Capsule, length=5, radius=2.25, shapeName=dyn_feather_3_capsule, center=<5.5472 0 0>, rotation=<rot -90 180 180>}]
 dyn_knife [{shapeType=Capsule, length=25.5, radius=2.25, shapeName=dyn_knife_capsule, center=<0.348 0 0>, rotation=<rot -90 0 0>}]
 dyn_sk_leather_01_r [{shapeType=Capsule, length=12, radius=10.75, shapeName=dyn_sk_leather_01_r_capsule, center=<-2.3615 -2 -11.5012>, rotation=<rot -90 180 180>}]
 dyn_sk_leather_02_r [{shapeType=Capsule, length=5.25, radius=6, shapeName=dyn_sk_leather_02_r_capsule, center=<6.5328 0 -0.9993>, rotation=<rot -90 0 0>}]
 pelvis [{shapeType=Capsule, length=12, radius=8.5, shapeName=pelvis_capsule, center=<-2.9573 -0.7623 0>}]
 spine_01 [{shapeType=Capsule, length=9.5, radius=9, shapeName=pelvis_capsule, center=<6.9758 0 0>, rotation=<rot 90 0 0>}]
 thigh_r [{shapeType=Capsule, length=15, radius=6.7, shapeName=thigh_r_capsule, center=<16.9375 0 0>, rotation=<rot -90 180 180>}]
}
constraints [
 {bone1Name=dyn_ape_toy, bone2Name=base_ape_toy, swing1LimitDegrees=40, swing2LimitDegrees=40, twistLimitDegrees=10}
 {bone1Name=dyn_bottle, bone2Name=base_bottle, swing1LimitDegrees=25, swing2LimitDegrees=45, twistLimitDegrees=10}
 {bone1Name=dyn_feather_1, bone2Name=base_feather_1, swing1LimitDegrees=45, swing2LimitDegrees=45, twistLimitDegrees=1, twistMotion=Locked}
 {bone1Name=dyn_feather_2, bone2Name=dyn_feather_1, swing1LimitDegrees=30, swing2LimitDegrees=45, twistLimitDegrees=1, twistMotion=Locked}
 {bone1Name=dyn_feather_3, bone2Name=dyn_feather_2, swing1LimitDegrees=25, swing2LimitDegrees=35, twistLimitDegrees=1, twistMotion=Locked}
 {bone1Name=dyn_knife, bone2Name=base_knife, swing1LimitDegrees=15, swing2LimitDegrees=15, twistLimitDegrees=10}
 {
  bone1Name=base_feather_1
  bone2Name=pelvis
  swing1LimitDegrees=45
  swing1Motion=Locked
  swing2LimitDegrees=45
  swing2Motion=Locked
  twistLimitDegrees=45
  twistMotion=Locked
 }
 {
  bone1Name=base_bottle
  bone2Name=pelvis
  swing1LimitDegrees=45
  swing1Motion=Locked
  swing2LimitDegrees=45
  swing2Motion=Locked
  twistLimitDegrees=45
  twistMotion=Locked
 }
 {
  bone1Name=base_ape_toy
  bone2Name=pelvis
  swing1LimitDegrees=45
  swing1Motion=Locked
  swing2LimitDegrees=45
  swing2Motion=Locked
  twistLimitDegrees=45
  twistMotion=Locked
 }
 {
  bone1Name=base_knife
  bone2Name=pelvis
  swing1LimitDegrees=45
  swing1Motion=Locked
  swing2LimitDegrees=45
  swing2Motion=Locked
  twistLimitDegrees=45
  twistMotion=Locked
 }
 {
  bone1Name=thigh_r
  bone2Name=pelvis
  swing1LimitDegrees=45
  swing1Motion=Locked
  swing2LimitDegrees=45
  swing2Motion=Locked
  twistLimitDegrees=45
  twistMotion=Locked
 }
 {
  bone1Name=calf_r
  bone2Name=thigh_r
  swing1LimitDegrees=45
  swing1Motion=Locked
  swing2LimitDegrees=45
  swing2Motion=Locked
  twistLimitDegrees=45
  twistMotion=Locked
 }
 {
  bone1Name=dyn_sk_leather_02_r
  bone2Name=dyn_sk_leather_01_r
  swing1LimitDegrees=45
  swing1Motion=Locked
  swing2LimitDegrees=45
  swing2Motion=Locked
  twistLimitDegrees=45
  twistMotion=Locked
 }
 {
  bone1Name=spine_01
  bone2Name=pelvis
  swing1LimitDegrees=45
  swing1Motion=Free
  swing2LimitDegrees=45
  swing2Motion=Free
  twistLimitDegrees=45
  twistMotion=Free
 }
 {bone1Name=dyn_feather_4, bone2Name=dyn_feather_3, swing1LimitDegrees=25, swing2LimitDegrees=30, twistLimitDegrees=1, twistMotion=Locked}
 {
  bone1Name=dyn_sk_leather_01_r
  bone2Name=pelvis
  swing1LimitDegrees=45
  swing1Motion=Free
  swing2LimitDegrees=45
  swing2Motion=Free
  twistLimitDegrees=45
  twistMotion=Free
 }
]
