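# Post-process-style AnimBP for the BoneMarrow female body part, animating its skirt, feather, ribbons and belt props over the copied parent pose.
# Control Rig and weighted constraints feed either AnimDynamics plus RigidBody simulation or a Trail fallback, selected by bRigidBodyEnabled.
# Additive bone offsets correct for capes and movement states like crouching, sliding and skydiving.
# The event graph updates alphas and gravity, and drives material WPO cloth parameters outside the front end.
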AnimBlueprint /Game/Characters/Player/Female/Medium/Bodies/F_MED_BoneMarrow/Meshes/F_MED_BoneMarrow_AnimBP
target_skeleton=&/F_MED_BoneMarrow_Skeleton
graphs [
 AnimGraph {nodes [
   n1 = "Transform (Modify) Bone - Bone: dyn_feather_2"(
    ComponentPose=Local To Component(LocalPose=Use cached pose 'cache_rigidBody'())
    Alpha=1
    Translation="0,0,0"
    Rotation="-30,0,0"
    Scale="1,0.8,0.8"
   ) {
    rotationMode=BMM_Additive
    rotationSpace=BCS_BoneSpace
    scaleMode=BMM_Additive
    scaleSpace=BCS_BoneSpace
    alphaBoolBlend {blendOption=Linear}
    boneToModify {boneName=dyn_feather_2}
    rotation {pitch=-30}
    scale {y=0.8, z=0.8}
   }
   n2 = Component To Local(ComponentPose=@n1)
   n3 = Use cached pose 'cache_rigidBody'()
   n4 = Get isWearingCape() -> (isWearingCape:bool)
   n5 = Get isCrouchingMoving() -> (isCrouchingMoving:double)
   n6 = "Transform (Modify) Bone - Bone: base_sk_leather_main"(
    ComponentPose=Local To Component(LocalPose=Blend Poses by bool(BlendPose_0=@n2, BlendPose_1=@n3, bActiveValue=@n4) {blendType=Linear})
    Alpha=@n5
    Translation="0,1.5,0.5"
    Rotation="0,0,0"
    Scale="1,1.05,1"
   ) {
    rotationMode=BMM_Additive
    rotationSpace=BCS_BoneSpace
    scaleMode=BMM_Additive
    scaleSpace=BCS_BoneSpace
    translationMode=BMM_Additive
    translationSpace=BCS_BoneSpace
    alphaBoolBlend {blendOption=Linear}
    alphaScaleBiasClamp {bInterpResult=true, interpSpeedDecreasing=5, interpSpeedIncreasing=5}
    boneToModify {boneName=base_sk_leather_main}
    scale {y=1.05}
    translation {y=1.5, z=0.5}
   }
   n7 = "Transform (Modify) Bone - Bone: dyn_sk_leather_main_01"(ComponentPose=@n6, Alpha=@n5, Translation="0,0,1", Rotation="-10,0,0", Scale="1,1,1") {
    rotationMode=BMM_Additive
    rotationSpace=BCS_BoneSpace
    scaleMode=BMM_Additive
    scaleSpace=BCS_BoneSpace
    translationMode=BMM_Additive
    translationSpace=BCS_BoneSpace
    alphaBoolBlend {blendOption=Linear}
    alphaScaleBiasClamp {bInterpResult=true, interpSpeedDecreasing=5, interpSpeedIncreasing=5}
    boneToModify {boneName=dyn_sk_leather_main_01}
    rotation {pitch=-10}
    translation {z=1}
   }
   n8 = "Transform (Modify) Bone - Bone: base_knife"(ComponentPose=@n7, Alpha=@n5, Translation="0,0,1", Rotation="0,-7,0", Scale="1,1,1") {
    rotationMode=BMM_Additive
    rotationSpace=BCS_BoneSpace
    scaleMode=BMM_Additive
    scaleSpace=BCS_BoneSpace
    translationMode=BMM_Additive
    translationSpace=BCS_BoneSpace
    alphaBoolBlend {blendOption=Linear}
    alphaScaleBiasClamp {bInterpResult=true}
    boneToModify {boneName=base_knife}
    rotation {yaw=-7}
    translation {z=1}
   }
   n9 = Get isSkirtSkydiving_alpha() -> (isSkirtSkydiving_alpha:double)
   n10 = "Transform (Modify) Bone - Bone: base_sk_leather_main"(ComponentPose=@n8, Alpha=@n9, Translation="-1,0,1.2", Rotation="5,0,0", Scale="1,1.1,1") {
    rotationMode=BMM_Additive
    rotationSpace=BCS_BoneSpace
    scaleMode=BMM_Additive
    scaleSpace=BCS_BoneSpace
    translationMode=BMM_Additive
    translationSpace=BCS_BoneSpace
    alphaBoolBlend {blendOption=Linear}
    alphaScaleBiasClamp {bInterpResult=true, interpSpeedDecreasing=5, interpSpeedIncreasing=5}
    boneToModify {boneName=base_sk_leather_main}
    rotation {pitch=5}
    scale {y=1.1}
    translation {x=-1, z=1.2}
   }
   n11 = "Transform (Modify) Bone - Bone: base_knife"(ComponentPose=@n10, Alpha=@n9, Translation="0,0,1", Rotation="0,15,-5", Scale="1,1,1") {
    rotationMode=BMM_Additive
    rotationSpace=BCS_BoneSpace
    scaleMode=BMM_Additive
    scaleSpace=BCS_BoneSpace
    translationMode=BMM_Additive
    translationSpace=BCS_BoneSpace
    alphaBoolBlend {blendOption=Linear}
    alphaScaleBiasClamp {bInterpResult=true}
    boneToModify {boneName=base_knife}
    rotation {roll=-5, yaw=15}
    translation {z=1}
   }
   n12 = "Transform (Modify) Bone - Bone: base_bottle"(ComponentPose=@n11, Alpha=@n9, Translation="0,0,0", Rotation="15,0,0", Scale="1,1,1") {
    rotationMode=BMM_Additive
    rotationSpace=BCS_BoneSpace
    scaleMode=BMM_Additive
    scaleSpace=BCS_BoneSpace
    translationMode=BMM_Additive
    translationSpace=BCS_BoneSpace
    alphaBoolBlend {blendOption=Linear}
    alphaScaleBiasClamp {bInterpResult=true}
    boneToModify {boneName=base_bottle}
    rotation {pitch=15}
   }
   n13 = "Transform (Modify) Bone - Bone: base_ring"(ComponentPose=@n12, Alpha=@n9, Translation="0,0,0", Rotation="45,0,0", Scale="1,1,1") {
    rotationMode=BMM_Additive
    rotationSpace=BCS_BoneSpace
    scaleMode=BMM_Additive
    scaleSpace=BCS_BoneSpace
    translationMode=BMM_Additive
    translationSpace=BCS_BoneSpace
    alphaBoolBlend {blendOption=Linear}
    alphaScaleBiasClamp {bInterpResult=true, interpSpeedDecreasing=5, interpSpeedIncreasing=5}
    boneToModify {boneName=base_ring}
    rotation {pitch=45}
   }
   n14 = "Transform (Modify) Bone - Bone: base_feather_1"(ComponentPose=@n13, Alpha=@n9, Translation="0,0,0", Rotation="35,-10,0", Scale="1,1,1") {
    rotationMode=BMM_Additive
    rotationSpace=BCS_BoneSpace
    scaleMode=BMM_Additive
    scaleSpace=BCS_BoneSpace
    translationMode=BMM_Additive
    translationSpace=BCS_BoneSpace
    alphaBoolBlend {blendOption=Linear}
    alphaScaleBiasClamp {bInterpResult=true}
    boneToModify {boneName=base_feather_1}
    rotation {pitch=35, yaw=-10}
   }
   n15 = "Transform (Modify) Bone - Bone: base_feather_1"(
    ComponentPose=@n14
    Alpha=Get is parachute opened alpha() -> (is parachute opened alpha:double)
    Translation="0,0,0"
    Rotation="-35,10,0"
    Scale="1,1,1"
   ) {
    rotationMode=BMM_Additive
    rotationSpace=BCS_BoneSpace
    scaleMode=BMM_Additive
    scaleSpace=BCS_BoneSpace
    translationMode=BMM_Additive
    translationSpace=BCS_BoneSpace
    alphaBoolBlend {blendOption=Linear}
    alphaScaleBiasClamp {bInterpResult=true, interpSpeedDecreasing=2, interpSpeedIncreasing=25}
    boneToModify {boneName=base_feather_1}
    rotation {pitch=-35, yaw=10}
   }
   n16 = "Transform (Modify) Bone - Bone: base_ribbon"(ComponentPose=@n15, Alpha=Get Head Up() -> (Head Up:double), Translation="0,0,4", Rotation="0,0,-90", Scale="1,1,1") {
    rotationMode=BMM_Additive
    rotationSpace=BCS_BoneSpace
    scaleMode=BMM_Additive
    scaleSpace=BCS_BoneSpace
    translationMode=BMM_Additive
    translationSpace=BCS_BoneSpace
    alphaBoolBlend {blendOption=Linear}
    boneToModify {boneName=base_ribbon}
    rotation {roll=-90}
    translation {z=4}
   }
   n17 = "Transform (Modify) Bone - Bone: base_knife"!(ComponentPose=@n16, bAlphaBoolEnabled:bool=Get bIsInFrontEnd(), Translation="0,0,0", Rotation="-12.5,0,0", Scale="1,1,1") -> (Pose) {
    alphaInputType=Bool
    rotationMode=BMM_Additive
    rotationSpace=BCS_BoneSpace
    scaleMode=BMM_Additive
    scaleSpace=BCS_BoneSpace
    translationMode=BMM_Additive
    translationSpace=BCS_BoneSpace
    alphaBoolBlend {blendOption=Linear}
    boneToModify {boneName=base_knife}
    rotation {pitch=-12.5}
   }
   Output Pose(Result=Component To Local(ComponentPose=@n17)) {layerGroup=None}
   n18 = Control Rig(Source=Copy Pose From Mesh() {bUseAttachedParent=true}) {
    alphaBoolBlend {blendOption=Linear}
    controlRigAssetReference {blueprintRigClass=&/F_MED_DeformRig_BoneMarrow_ControlRig.F_MED_DeformRig_BoneMarrow_ControlRig_C}
   }
   n19 = "Constraint - base_ape_toy (PPPPPPPPPP)"!(
    ComponentPose=Local To Component(LocalPose=@n18)
    Alpha=1
    ConstraintWeights_0=0.5488
    ConstraintWeights_1:float=0.318
    ConstraintWeights_2:float=0.0913
    ConstraintWeights_3:float=0.0244
    ConstraintWeights_4:float=0.0046
    ConstraintWeights_5:float=0.0039
    ConstraintWeights_6:float=0.0037
    ConstraintWeights_7:float=0.0036
    ConstraintWeights_8:float=0.0012
    ConstraintWeights_9:float=0.0005
   ) -> (Pose) {
    alphaBoolBlend {blendOption=Linear}
    boneToModify {boneName=base_ape_toy}
    constraintSetup [
     {targetBone {boneName=pelvis}}
     {targetBone {boneName=thigh_r}}
     {targetBone {boneName=deform_groin_r}}
     {targetBone {boneName=spine_01}}
     {targetBone {boneName=deform_glute_r}}
     {targetBone {boneName=thigh_l}}
     {targetBone {boneName=thigh_twist_01_r}}
     {targetBone {boneName=spine_02}}
     {targetBone {boneName=deform_groin_l}}
     {targetBone {boneName=spine_03}}
    ]
    constraintWeights [1,1,1,1,1,1,1,1,1,1]
   }
   n20 = "Constraint - base_bottle (P)"!(ComponentPose=@n19, Alpha=1, ConstraintWeights_0=1, ConstraintWeights_1:float=0) -> (Pose) {
    alphaBoolBlend {blendOption=Linear}
    boneToModify {boneName=base_bottle}
    constraintSetup [{targetBone {boneName=pelvis}}]
    constraintWeights [1,0]
   }
   n21 = "Constraint - base_feather_1 (P)"!(ComponentPose=@n20, Alpha=1, ConstraintWeights_0=1, ConstraintWeights_1:float=0) -> (Pose) {
    alphaBoolBlend {blendOption=Linear}
    boneToModify {boneName=base_feather_1}
    constraintSetup [{targetBone {boneName=pelvis}}]
    constraintWeights [1,0]
   }
   n22 = "Constraint - base_knife (PPPPPPPPPPP)"!(
    ComponentPose=@n21
    Alpha=1
    ConstraintWeights_0=0.6999
    ConstraintWeights_1:float=0.1149
    ConstraintWeights_2:float=0.072
    ConstraintWeights_3:float=0.0714
    ConstraintWeights_4:float=0.0321
    ConstraintWeights_5:float=0.0058
    ConstraintWeights_6:float=0.0018
    ConstraintWeights_7:float=0.0014
    ConstraintWeights_8:float=0.0004
    ConstraintWeights_9:float=0.0003
    ConstraintWeights_10:float=0.0001
   ) -> (Pose) {
    alphaBoolBlend {blendOption=Linear}
    boneToModify {boneName=base_knife}
    constraintSetup [
     {targetBone {boneName=pelvis}}
     {targetBone {boneName=spine_01}}
     {targetBone {boneName=deform_groin_r}}
     {targetBone {boneName=thigh_r}}
     {targetBone {boneName=spine_02}}
     {targetBone {boneName=spine_03}}
     {targetBone {boneName=thigh_l}}
     {targetBone {boneName=deform_glute_r}}
     {targetBone {boneName=deform_groin_l}}
     {targetBone {boneName=thigh_twist_01_r}}
     {targetBone {boneName=spine_04}}
    ]
    constraintWeights [1,1,1,1,1,1,1,1,1,1,1]
   }
   n23 = "Constraint - base_ring (PPPPPPP)"!(
    ComponentPose=@n22
    Alpha=1
    ConstraintWeights_0=0.5351
    ConstraintWeights_1:float=0.2716
    ConstraintWeights_2:float=0.1143
    ConstraintWeights_3:float=0.0567
    ConstraintWeights_4:float=0.0203
    ConstraintWeights_5:float=0.0011
    ConstraintWeights_6:float=0.0009
   ) -> (Pose) {
    alphaBoolBlend {blendOption=Linear}
    boneToModify {boneName=base_ring}
    constraintSetup [
     {targetBone {boneName=pelvis}}
     {targetBone {boneName=spine_01}}
     {targetBone {boneName=spine_02}}
     {targetBone {boneName=spine_03}}
     {targetBone {boneName=spine_04}}
     {targetBone {boneName=thigh_l}}
     {targetBone {boneName=thigh_r}}
    ]
    constraintWeights [1,1,1,1,1,1,1]
   }
   n24 = "Constraint - base_ribbon (PPPPPP)"!(
    ComponentPose=@n23
    Alpha=1
    ConstraintWeights_0=0.515
    ConstraintWeights_1:float=0.3563
    ConstraintWeights_2:float=0.081
    ConstraintWeights_3:float=0.045
    ConstraintWeights_4:float=0.0013
    ConstraintWeights_5:float=0.0013
   ) -> (Pose) {
    alphaBoolBlend {blendOption=Linear}
    boneToModify {boneName=base_ribbon}
    constraintSetup [
     {targetBone {boneName=neck_01}}
     {targetBone {boneName=spine_05}}
     {targetBone {boneName=neck_02}}
     {targetBone {boneName=head}}
     {targetBone {boneName=clavicle_l}}
     {targetBone {boneName=clavicle_r}}
    ]
    constraintWeights [1,1,1,1,1,1]
   }
   Save cached pose 'cache constraint'(Pose=Component To Local(ComponentPose=@n24)) {cacheName=cache constraint}
   n25 = Anim Dynamics - Bone: dyn_ribbon_l(ComponentPose=Local To Component(LocalPose=Use cached pose 'PreSim_Cache'()), Alpha=0.5, ExternalForce="0,0,0") {
    angularDampingOverride=1
    angularSpringConstant=20000
    bAngularSpring=true
    bOverrideAngularDamping=true
    bUseGravityOverride=true
    gravityScale=0.35
    simulationSpace=RootRelative
    alphaBoolBlend {blendOption=Linear}
    boundBone {boneName=dyn_ribbon_l}
    componentLinearAccScale=<1 1 1>
    componentLinearVelScale=<1 1 1>
    gravityOverride {z=-980}
    physicsBodyDefinitions [{
      boundBone {boneName=dyn_ribbon_l}
      boxExtents=<20 30 30>
      constraintSetup {angularLimitsMax=<10 10 10>, angularLimitsMin=<-10 -10 -10>, angularTarget=<1 0 0>}
      localJointOffset=<4 0 0>
     }]
   }
   n26 = Anim Dynamics - Bone: dyn_ribbon_r(ComponentPose=@n25, Alpha=0.5, ExternalForce="0,0,0") {
    angularDampingOverride=1
    angularSpringConstant=20000
    bAngularSpring=true
    bOverrideAngularDamping=true
    bUseGravityOverride=true
    gravityScale=0.35
    simulationSpace=RootRelative
    alphaBoolBlend {blendOption=Linear}
    boundBone {boneName=dyn_ribbon_r}
    componentLinearAccScale=<1 1 1>
    componentLinearVelScale=<1 1 1>
    gravityOverride {z=-980}
    physicsBodyDefinitions [{
      boundBone {boneName=dyn_ribbon_r}
      boxExtents=<20 30 30>
      constraintSetup {angularLimitsMax=<10 10 10>, angularLimitsMin=<-10 -10 -10>, angularTarget=<1 0 0>}
      localJointOffset=<4 0 0>
     }]
   }
   n27 = Anim Dynamics - Bone: dyn_ring!(ComponentPose=@n26, Alpha=1, GravityOverride:Vector=Get GravityOverride(), ExternalForce="0,0,0") -> (Pose) {
    angularDampingOverride=1
    angularSpringConstant=20000
    bAngularSpring=true
    bOverrideAngularDamping=true
    bUseGravityOverride=true
    gravityScale=0.35
    simulationSpace=RootRelative
    alphaBoolBlend {blendOption=Linear}
    boundBone {boneName=dyn_ring}
    componentLinearAccScale=<1 1 1>
    componentLinearVelScale=<1 1 1>
    gravityOverride {z=-980}
    physicsBodyDefinitions [{
      boundBone {boneName=dyn_ring}
      boxExtents=<20 30 30>
      constraintSetup {angularLimitsMax=<20 30 20>, angularLimitsMin=<-20 -20 -20>, angularTarget=<1 0 0>}
      localJointOffset=<4 0 0>
     }]
   }
   n28 = RigidBody!(
    ComponentPose=@n27
    Alpha=Get SkirtRB_alpha() -> (SkirtRB_alpha:double)
    OverrideWorldGravity:Vector=Get GravityOverride()
    ExternalForce
    ComponentLinearAccScale:Vector="0.12,0.24,0.24"
    ComponentLinearVelScale:Vector="0.12,0.24,0.24"
   ) -> (Pose) {
    bClampLinearTranslationLimitToRefPose=true
    bOverrideWorldGravity=true
    overridePhysicsAsset=&/F_MED_BoneMarrow_Physics_Tail_skirt
    simulationSpace=BaseBoneSpace
    alphaBoolBlend {blendOption=Linear}
    alphaScaleBiasClamp {bInterpResult=true}
    baseBoneRef {boneName=pelvis}
    componentAppliedLinearAccClamp=<3000 3000 3000>
    componentLinearAccScale=<0.4 0.4 0.4>
    componentLinearVelScale=<0.4 0.4 0.4>
    simSpaceSettings {worldAlpha=0.2}
   }
   n29 = RigidBody!(
    ComponentPose=@n28
    Alpha=Get PropsRB_alpha() -> (PropsRB_alpha:double)
    OverrideWorldGravity=Get GravityOverride()
    ExternalForce
    ComponentLinearAccScale="0.2,0.2,0.2"
   ) -> (Pose) {
    bClampLinearTranslationLimitToRefPose=true
    bOverrideWorldGravity=true
    overridePhysicsAsset=&/F_MED_BoneMarrow_Physics_props
    simulationSpace=BaseBoneSpace
    alphaBoolBlend {blendOption=Linear}
    baseBoneRef {boneName=pelvis}
    componentAppliedLinearAccClamp=<3000 3000 3000>
    componentLinearAccScale=<0.6 0.6 0.6>
    componentLinearVelScale=<0.2 0.2 0.2>
    simSpaceSettings {worldAlpha=0.2}
   }
   n30 = "Transform (Modify) Bone - Bone: base_sk_leather_main"(
    ComponentPose=Local To Component(LocalPose=Use cached pose 'PreSim_Cache'())
    Alpha=Get is sprinting alpha() -> (is sprinting alpha:double)
    Translation="-1,0,1.2"
    Rotation="10,0,0"
    Scale="1,1.1,1"
   ) {
    rotationMode=BMM_Additive
    rotationSpace=BCS_BoneSpace
    scaleMode=BMM_Additive
    scaleSpace=BCS_BoneSpace
    translationMode=BMM_Additive
    translationSpace=BCS_BoneSpace
    alphaBoolBlend {blendOption=Linear}
    alphaScaleBiasClamp {bInterpResult=true, interpSpeedDecreasing=5, interpSpeedIncreasing=5}
    boneToModify {boneName=base_sk_leather_main}
    rotation {pitch=10}
    scale {y=1.1}
    translation {x=-1, z=1.2}
   }
   n31 = Trail controller - Bone: dyn_sk_leather_main_02!(ComponentPose=@n30, Alpha=Get SkirtTrail_alpha() -> (SkirtTrail_alpha:double), RelaxationSpeedScale:float=4, FakeVelocity:Vector="0,0,0") -> (Pose) {
    bActorSpaceFakeVel=true
    bShowTrailLocation=true
    bUsePlanarLimit=true
    debugLifeTime=0.1
    maxDeltaTime=0.0333
    relaxationSpeedScale=1.2
    alphaScaleBiasClamp {bInterpResult=true, interpSpeedDecreasing=25, interpSpeedIncreasing=25}
    baseJoint {boneName=root}
    rotationLimits [{limitMax=<180 180 180>}, {limitMax=<180 180 180>}]
    rotationOffsets [<0 0 0>, <0 0 0>]
    trailBone {boneName=dyn_sk_leather_main_02}
    trailRelaxationSpeed {editorCurveData {keys [<at 0 15 RCIM_Linear>,<at 1 6 RCIM_Linear>]}}
   }
   n32 = Trail controller - Bone: dyn_sk_leather_03_l!(ComponentPose=@n31, Alpha=0.4, RelaxationSpeedScale:float=4, FakeVelocity:Vector="0,0,0") -> (Pose) {
    bActorSpaceFakeVel=true
    bShowTrailLocation=true
    bUsePlanarLimit=true
    chainLength=3
    debugLifeTime=0.1
    maxDeltaTime=0.0333
    relaxationSpeedScale=1.2
    alphaScaleBiasClamp {interpSpeedDecreasing=25, interpSpeedIncreasing=25}
    baseJoint {boneName=root}
    rotationLimits [{limitMax=<180 180 180>}, {limitMax=<180 180 180>}, {limitMax=<180 180 180>}]
    rotationOffsets [<0 0 0>, <0 0 0>, <0 0 0>]
    trailBone {boneName=dyn_sk_leather_03_l}
    trailRelaxationSpeed {editorCurveData {keys [<at 0 15 RCIM_Linear>,<at 1 6 RCIM_Linear>]}}
   }
   n33 = Trail controller - Bone: dyn_sk_leather_03_r!(ComponentPose=@n32, Alpha=0.4, RelaxationSpeedScale:float=4, FakeVelocity:Vector="0,0,0") -> (Pose) {
    bActorSpaceFakeVel=true
    bShowTrailLocation=true
    bUsePlanarLimit=true
    chainLength=3
    debugLifeTime=0.1
    maxDeltaTime=0.0333
    relaxationSpeedScale=1.2
    alphaScaleBiasClamp {interpSpeedDecreasing=25, interpSpeedIncreasing=25}
    baseJoint {boneName=root}
    rotationLimits [{limitMax=<180 180 180>}, {limitMax=<180 180 180>}, {limitMax=<180 180 180>}]
    rotationOffsets [<0 0 0>, <0 0 0>, <0 0 0>]
    trailBone {boneName=dyn_sk_leather_03_r}
    trailRelaxationSpeed {editorCurveData {keys [<at 0 15 RCIM_Linear>,<at 1 6 RCIM_Linear>]}}
   }
   n34 = "Transform (Modify) Bone - Bone: base_feather_1"(ComponentPose=@n33, Alpha=1, Translation="0,0,0", Rotation="-25,0,0", Scale="1,1,1") {
    rotationMode=BMM_Additive
    rotationSpace=BCS_BoneSpace
    scaleMode=BMM_Additive
    scaleSpace=BCS_BoneSpace
    translationMode=BMM_Additive
    translationSpace=BCS_BoneSpace
    alphaBoolBlend {blendOption=Linear}
    boneToModify {boneName=base_feather_1}
    rotation {pitch=-25}
   }
   n35 = Trail controller - Bone: dyn_feather_3!(ComponentPose=@n34, Alpha=Get FeatherTrail_alpha() -> (FeatherTrail_alpha:double), RelaxationSpeedScale:float=6, FakeVelocity:Vector="0,0,0") -> (Pose) {
    bActorSpaceFakeVel=true
    bShowTrailLocation=true
    bUsePlanarLimit=true
    chainLength=3
    debugLifeTime=0.1
    maxDeltaTime=0.0333
    relaxationSpeedScale=1.2
    alphaScaleBiasClamp {bInterpResult=true, interpSpeedDecreasing=25, interpSpeedIncreasing=25}
    baseJoint {boneName=root}
    rotationLimits [{limitMax=<180 180 180>}, {limitMax=<180 180 180>}, {limitMax=<180 180 180>}]
    rotationOffsets [<0 0 0>, <0 0 0>, <0 0 0>]
    trailBone {boneName=dyn_feather_3}
    trailRelaxationSpeed {editorCurveData {keys [<at 0 15 RCIM_Linear>,<at 1 6 RCIM_Linear>]}}
   }
   Save cached pose 'cache_rigidBody'(Pose=Blend Poses by bool(BlendPose_0=Component To Local(ComponentPose=@n29), BlendPose_1=Component To Local(ComponentPose=@n35), bActiveValue=Get bRigidBodyEnabled()) {blendType=Linear}) {cacheName=cache_rigidBody}
   n36 = "Constraint - base_sk_leather_main (RR)"!(
    ComponentPose=Local To Component(LocalPose=Use cached pose 'cache constraint'())
    Alpha=1
    ConstraintWeights_0=1
    ConstraintWeights_1:float=1
   ) -> (Pose) {
    alphaBoolBlend {blendOption=Linear}
    boneToModify {boneName=base_sk_leather_main}
    constraintSetup [{transformType=Rotation, targetBone {boneName=thigh_l}}, {transformType=Rotation, targetBone {boneName=thigh_r}}]
    constraintWeights [1,1]
   }
   n37 = "Constraint - base_bottle (R)"(ComponentPose=@n36, Alpha=1, ConstraintWeights_0=1) {
    alphaBoolBlend {blendOption=Linear}
    boneToModify {boneName=base_bottle}
    constraintSetup [{transformType=Rotation, targetBone {boneName=thigh_l}}]
    constraintWeights [1]
   }
   n38 = "Constraint - base_feather_1 (RR)"!(ComponentPose=@n37, Alpha=1, ConstraintWeights_0=1, ConstraintWeights_1:float=1) -> (Pose) {
    alphaBoolBlend {blendOption=Linear}
    boneToModify {boneName=base_feather_1}
    constraintSetup [{transformType=Rotation, targetBone {boneName=thigh_r}}, {transformType=Rotation, targetBone {boneName=pelvis}}]
    constraintWeights [1,1]
   }
   n39 = Component To Local(ComponentPose=@n38)
   n40 = Use cached pose 'cache constraint'()
   n41 = Get bIsParachuteOpened()
   n42 = "Transform (Modify) Bone - Bone: base_sk_leather_main"(
    ComponentPose=Local To Component(LocalPose=Blend Poses by bool(BlendPose_0=@n39, BlendPose_1=@n40, bActiveValue=@n41) {blendType=Linear})
    Alpha="Get is sliding float "() -> ("is sliding float ":double)
    Translation="2,0,0"
    Rotation="120,0,0"
    Scale="1,1,1"
   ) {
    rotationMode=BMM_Additive
    rotationSpace=BCS_BoneSpace
    alphaBoolBlend {blendOption=Linear}
    alphaScaleBiasClamp {bInterpResult=true, interpSpeedDecreasing=25, interpSpeedIncreasing=25}
    boneToModify {boneName=base_sk_leather_main}
    rotation {pitch=120}
    translation {x=2}
   }
   n43 = "Transform (Modify) Bone - Bone: base_feather_1"(ComponentPose=@n42, Alpha=Get is sprinting alpha(), Translation="0,0,0", Rotation="10,-10,0", Scale="1,1,1") {
    rotationMode=BMM_Additive
    rotationSpace=BCS_BoneSpace
    scaleMode=BMM_Additive
    scaleSpace=BCS_BoneSpace
    translationMode=BMM_Additive
    translationSpace=BCS_BoneSpace
    alphaBoolBlend {blendOption=Linear}
    alphaScaleBiasClamp {bInterpResult=true}
    boneToModify {boneName=base_feather_1}
    rotation {pitch=10, yaw=-10}
   }
   n44 = "Transform (Modify) Bone - Bone: base_sk_leather_main"(ComponentPose=@n43, Alpha=Get isNotMoving() -> (isNotMoving:double), Translation="0,0,0", Rotation="0,-7.5,0", Scale="1,1,1") {
    rotationMode=BMM_Additive
    rotationSpace=BCS_BoneSpace
    scaleMode=BMM_Additive
    scaleSpace=BCS_BoneSpace
    translationMode=BMM_Additive
    translationSpace=BCS_BoneSpace
    alphaBoolBlend {blendOption=Linear}
    boneToModify {boneName=base_sk_leather_main}
    rotation {yaw=-7.5}
   }
   Save cached pose 'PreSim_Cache'(Pose=Component To Local(ComponentPose=@n44)) {cacheName=PreSim_Cache}
  ]}
 EventGraph {nodes [
   Event OnCharacterCustomizationLoaded()
   > n1 = Set SkelMesh(SkelMesh=GetPartSkeletalMeshComponent())
   > n2 = Get Dynamic Material Instance(SkeletalMeshComponent=@n1)
   > Set Mat Instance Dynamic(Mat Instance Dynamic=@n2)
   > n3 = GetOrCreateDynamicOverlayMaterial(MeshComponent=@n1, MaterialName=None)
   > Set Overlay Mat Instance Dynamic(Overlay Mat Instance Dynamic=@n3)
   n4 = Event BlueprintUpdateAnimation()
   > n5 = Sequence!(execute) -> (then_0, then_1, then_2:Exec, then_3:Exec, then_4:Exec, then_5:Exec)
   > n6 = Gravity Override State Function DEPRECATED(CharacterPartAnimInstance=Self-Reference(), GameTimeInSeconds=GetGameTimeInSeconds(), GravityOverrideParams=Get Gravity Override Params())
   > Set GravityOverride(GravityOverride=@n6)
   > Set isSkirtSkydiving_alpha(isSkirtSkydiving_alpha:double="To Float (Boolean)"(InBool=Get bIsSkyDiving())) -> (Output_Get:double)
   > Set is parachute opened alpha(is parachute opened alpha:double="To Float (Boolean)"(InBool=Get bIsParachuteOpened())) -> (Output_Get:double)
   n7 = Get bIsMoving2D()
   n8 = Get bIsCrouching()
   > Set isCrouchingMoving(isCrouchingMoving:double="To Float (Boolean)"(InBool=AND Boolean(A=@n8, B=@n7))) -> (Output_Get:double)
   n9 = NOT Boolean(A=Get bIsMoving2D())
   n10 = NOT Boolean(A=Get bIsCrouching())
   n11 = NOT Boolean(A=Get bIsSkyDiving())
   > Set isNotMoving(isNotMoving:double="To Float (Boolean)"(InBool=AND Boolean!(A=@n9, B=@n11, C:bool=@n10) -> (ReturnValue))) -> (Output_Get:double)
   n12 = Get bIsTacticalSprinting()
   n13 = Get bIsSprinting()
   > Set is sprinting alpha(is sprinting alpha:double="To Float (Boolean)"(InBool=OR Boolean(A=@n13, B=@n12))) -> (Output_Get:double)
   > Set is jumping alpha(is jumping alpha:double="To Float (Boolean)"(InBool=Get bIsJumping())) -> (Output_Get:double)
   > "Set is sliding float "("is sliding float ":double="To Float (Boolean)"(InBool=Get bIsSliding())) -> (Output_Get:double)
   Is Valid(exec=@n5.then_3, InputObject=Get PlayerPawn() -> (PlayerPawn:Fort Player Pawn Athena Object Reference))
   > n14 = Branch(Condition=Get bIsInFrontEnd())
   Set WPO Globals Function(
    execute=@n14.else
    Material=Get Mat Instance Dynamic()
    Overlay Material=Get Overlay Mat Instance Dynamic()
    Skeletal Mesh=Get SkelMesh()
    "Local Noise Transform  Location"=Get Local Noise Transform Location() -> (Local Noise Transform Location:Vector)
    Local Noise Offset=Get Local Noise Offset()
   )
   > Set WPO Regional Function!(
    execute
    Material:Material Instance Dynamic Object Reference=Get Mat Instance Dynamic()
    Overlay Material:Material Instance Dynamic Object Reference=Get Overlay Mat Instance Dynamic()
    Skeletal Mesh:Skeletal Mesh Component Object Reference=Get SkelMesh()
    Velocity Transform:Name=foot_r
    Velocity Interp Speed:double=Get Velocity Interp Speed()
    Collision Transform:Name=calf_r
    Collision Transform Axis:EAxis Enum
    Invert Collision Transform Axis:bool=false
    Delta Time X:double=@n4.DeltaTimeX
    Previous Velocity Transform Location:"Vector (by ref)"=Get R1 Velocity Transform Location()
    Previous Velocity Transform Velocity:"Vector (by ref)"=Get R1 Velocity Transform Velocity()
    Region:int=1
   ) -> (then)
   > Set WPO Regional Function(
    Material=Get Mat Instance Dynamic()
    Overlay Material=Get Overlay Mat Instance Dynamic()
    Skeletal Mesh=Get SkelMesh()
    Velocity Transform=foot_l
    Velocity Interp Speed=Get Velocity Interp Speed()
    Collision Transform=calf_l
    Invert Collision Transform Axis=true
    Delta Time X=@n4.DeltaTimeX
    Previous Velocity Transform Location=Get R2 Velocity Transform Location()
    Previous Velocity Transform Velocity=Get R2 Velocity Transform Velocity()
    Region=2
   )
   > n15 = Branch(Condition=Get bIsSkyDiving())
   > Set WPO Pho Wind Intensity(WPO Pho Wind Intensity=1)
   > Set WPO Pho Turbulence(WPO Pho Turbulence:double=1) -> (Output_Get:double)
   > n16 = Set WPO Pho Drag(WPO Pho Drag:double=0.4) -> (Output_Get:double)
   SetScalarParameterValue(execute=[@n19.then, @n16.then], self=Get Mat Instance Dynamic(), ParameterName=ClothSim_WindStrength, Value=Get WPO Pho Wind Intensity())
   n17 = Get WPO Pho Turbulence() -> (WPO Pho Turbulence:double)
   > SetScalarParameterValue(self=Get Mat Instance Dynamic(), ParameterName=ClothSim_R1_TurbulenceStrength, Value=@n17)
   > SetScalarParameterValue(self=Get Mat Instance Dynamic(), ParameterName=ClothSim_R2_TurbulenceStrength, Value=@n17)
   n18 = Get WPO Pho Drag() -> (WPO Pho Drag:double)
   > SetScalarParameterValue(self=Get Mat Instance Dynamic(), ParameterName=ClothSim_R1_DragStrength, Value=@n18)
   > SetScalarParameterValue(self=Get Mat Instance Dynamic(), ParameterName=ClothSim_R2_DragStrength, Value=@n18)
   Set WPO Pho Wind Intensity(execute=@n15.else, WPO Pho Wind Intensity=0)
   > Set WPO Pho Turbulence(WPO Pho Turbulence=0.2)
   > n19 = Set WPO Pho Drag(WPO Pho Drag=8)
   Set Head Up(
    execute=@n5.then_4
    Head Up:double=GetAlphaFromReferenceTransformDelta(SkeletalMeshComponent=Get SkelMesh(), BoneName=spine_05, BaseName=neck_01, InputRange=(X=0,Y=-55), TwistAxis="0,0,1")
   ) -> (Output_Get:double)
   n20 = CheckForBackbling()
   Set isWearingBackbling(execute=@n5.then_2, isWearingBackbling=@n20)
   > Set isWearingCape(isWearingCape=@n20.bIsBackblingLong)
   > Branch(Condition=Get isWearingCape())
   > Set SkirtRB_alpha(SkirtRB_alpha:double=float * float(A=Get SkirtRB_alpha(), B=0.65) {is=title K2Node_PromotableOperator}) -> (Output_Get:double)
   > Set PropsRB_alpha(PropsRB_alpha:double=float * float(A=Get PropsRB_alpha(), B=0.4)) -> (Output_Get:double)
   n21 = Branch(execute=@n5.then_1, Condition=Get bIsSkyDiving())
   > Set SkirtRB_alpha(SkirtRB_alpha=0.4)
   > Set PropsRB_alpha(PropsRB_alpha=0.6)
   > Set SkirtTrail_alpha(SkirtTrail_alpha:double=0) -> (Output_Get:double)
   > Set FeatherTrail_alpha(FeatherTrail_alpha:double=0.25) -> (Output_Get:double)
   Set SkirtRB_alpha(execute=@n21.else, SkirtRB_alpha=1)
   > Set PropsRB_alpha(PropsRB_alpha=1)
   > Set SkirtTrail_alpha(SkirtTrail_alpha=0.15)
   > Set FeatherTrail_alpha(FeatherTrail_alpha=0.5)
   Event BlueprintInitializeAnimation()
   > Set SkelMesh(SkelMesh=GetOwnerSkeletalMeshComponent())
   n22 = TryGetPawnOwner()
   > Is Valid(InputObject=@n22)
   > n23 = Cast To FortPlayerPawnAthena(Object=@n22)
   > Set PlayerPawn(PlayerPawn:Fort Player Pawn Athena Object Reference=@n23) -> (Output_Get:Fort Player Pawn Athena Object Reference)
  ]}
]
variables [
 Body_Mesh {type=SkeletalMeshComponent}
 BodyMesh {type=SkeletalMeshComponent}
 FeatherTrail_alpha {type=number}
 Gravity Override Params {type=GravityOverrideParamsStruct}
 GravityOverride {type=Vector}
 Head Up {type=number}
 is jumping alpha {type=number}
 is parachute opened alpha {type=number}
 is size backskirt_sprint+jump {type=number}
 "is sliding float " {type=number}
 is sprinting alpha {type=number}
 isCrouchingMoving {type=number}
 isNotMoving {type=number}
 isSkirtSkydiving_alpha {type=number}
 isWearingBackbling {type=boolean}
 isWearingCape {type=boolean}
 linear_velocity_scale_jump {type=Vector}
 Local Noise Offset {type=Vector}
 Local Noise Transform Location {type=Vector}
 Mat Instance Dynamic {type=MaterialInstanceDynamic}
 Overlay Mat Instance Dynamic {type=MaterialInstanceDynamic}
 PlayerPawn {type=FortPlayerPawnAthena}
 PropsRB_alpha {type=number}
 R1 Velocity Transform Location {type=Vector}
 R1 Velocity Transform Velocity {type=Vector}
 R2 Velocity Transform Location {type=Vector}
 R2 Velocity Transform Velocity {type=Vector}
 skel mesh {type=SkeletalMeshComponent}
 SkelMesh {type=SkeletalMeshComponent}
 SkirtRB_alpha {type=number}
 SkirtTrail_alpha {type=number}
 skydiving_floatDive {type=boolean}
 TailMesh {type=SkeletalMeshComponent}
 Velocity Interp Speed {type=number}
 WPO Pho Drag {type=number}
 WPO Pho Turbulence {type=number}
 WPO Pho Wind Intensity {type=number}
]
